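# UEFN_InputModeWidgetSwitcher is a UMG widget built on the native UEFNInputModeWidgetSwitcher class, providing a container for content that differs by input mode.
# Its inherited root WidgetSwitcher, Switcher_InputMode, holds three named slots: MouseAndKeyboard, Gamepad and Touch.
# The Blueprint adds no widgets, variables or graph logic, so any switching between slots is left to the parent class.
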WidgetBlueprint /Game/Valkyrie/UMG/UEFN_InputModeWidgetSwitcher
graphs [EventGraph {links [], nodes []}]
variables []
widgets {
 info {
  inheritedWidgetCount=4
  parentClass=/Script/FortniteGame.UEFNInputModeWidgetSwitcher
  rootWidgetClass=&WidgetSwitcher
  widgetCount=4
 }
 widgets [
  Switcher_InputMode {widgetClassPath=&WidgetSwitcher, bInherited=true, widget=Switcher_InputMode}
  MouseAndKeyboard {
   widgetClassPath=&NamedSlot
   bInherited=true
   bIsVariable=true
   parent=Switcher_InputMode
   widget=MouseAndKeyboard
   slot {class=WidgetSwitcherSlot}
  }
  Gamepad {
   widgetClassPath=&NamedSlot
   bInherited=true
   bIsVariable=true
   parent=Switcher_InputMode
   widget=Gamepad
   slot {class=WidgetSwitcherSlot}
  }
  Touch {
   widgetClassPath=&NamedSlot
   bInherited=true
   bIsVariable=true
   parent=Switcher_InputMode
   widget=Touch
   slot {class=WidgetSwitcherSlot}
  }
 ]
}
