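# UEFN_Slider is a UMG slider widget that shows a numeric setting as a filled bar with centered text.
# A transparent AnalogSlider, also the gamepad focus target, overlays a material progress bar and text, driving both with normalized progress and formatted value on Construct and on change.
# Non-touch hover focuses a UIKit outline and plays a sound.
# At runtime PreConstruct copies a nonzero slider value into the broadcasting Value variable.
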WidgetBlueprint /Game/Valkyrie/UMG/UEFN_Slider
graphs [
 EventGraph {nodes [
   "On Value Changed (Slider_SettingValue)"!() -> (OutputDelegate, then, Value:float) {is=title K2Node_ComponentBoundEvent}
   n1 = Get Slider_SettingValue() -> (Slider_SettingValue:Analog Slider Object Reference)
   n2 = Get MinValue!(self:Slider Object Reference=@n1) -> (MinValue:float)
   n3 = Get Value!(self:Slider Object Reference=@n1) -> (Value:float)
   n4 = SafeDivide(A=float - float(A=@n3, B=@n2), B=float - float(A=Get MaxValue!(self:Slider Object Reference=@n1) -> (MaxValue:float), B=@n2))
   n5 = Get Progress() -> (Progress:Simple Material Progress Bar Object Reference)
   > SetProgress!(execute, self:Fort Simple Material Progress Bar Object Reference=@n5, Progress:float=@n4) -> (then)
   n6 = "To Text (Float)"(Value=@n3, MinimumIntegralDigits=1, MaximumIntegralDigits=324, MinimumFractionalDigits=0, MaximumFractionalDigits=2)
   > "SetText (Text)"(self=Get Text_SettingValue() -> (Text_SettingValue:Common Text Object Reference), InText=@n6)
   Event Construct()
   > SetProgress(self=@n5, Progress=@n4)
   > "SetText (Text)"(self=Get Text_SettingValue(), InText=@n6)
   Event OnMouseEnter()
   n7 = IsUsingTouch(PlayerController=GetOwningPlayer())
   > n8 = Branch(Condition=@n7)
   n9 = Get Block_Outline() -> (Block_Outline:WBP UIKit Block Outline Object Reference)
   RequestTransition!(execute=@n8.else, self:UIKit Block Object Reference=@n9, State:EUIKitBlockVisualState Enum=Focused) -> (then)
   > PlaySound2D(Sound=Get Sound() -> (Sound:Sound Base Object Reference))
   Event OnMouseLeave!() -> (OutputDelegate, then, MouseEvent:"Pointer Event Structure (by ref)") {is=title K2Node_Event}
   > n10 = Branch(Condition=@n7)
   RequestTransition(execute=@n10.else, self=@n9, State=Unfocused)
   n11 = Event PreConstruct()
   > n12 = Branch(Condition=@n11.IsDesignTime)
   n13 = Branch(execute=@n12.else, Condition="Equal (Float)"(A=@n3))
   Set with Broadcast Value!(execute=@n13.else, Value:double=@n3) -> (then, Output_Get:double)
  ]}
 GetPrimaryGamepadFocusWidget {nodes [
   GetPrimaryGamepadFocusWidget() {is=title K2Node_FunctionEntry}
   > Return Node!(execute, ReturnValue:Widget Object Reference=Get Slider_SettingValue())
  ]}
]
variables [
 Sound {type=SoundBase, default=/Game/Sounds/Fort_UI_Sounds/FrontEnd/New_FrontEnd/Fortnite_UI_MainMenu_Tab_Hover_01_Cue}
 Value {type=number}
]
widgets {
 info {inheritedWidgetCount=0, parentClass=&UserWidget, rootWidgetClass=&Overlay, widgetCount=8}
 widgets [
  Overlay_1 {widgetClassPath=&Overlay, widget=Overlay_1}
  Block_Outline {
   widgetClassPath=/FortUILibrary/UIKit/Buttons/Blocks/WBP_UIKit_Block_Outline
   bIsVariable=true
   parent=Overlay_1
   widget=Block_Outline
   properties {bottomLeftCornerRadius=0.113, bottomRightCornerRadius=0.113, topLeftCornerRadius=0.113, topRightCornerRadius=0.113}
   slot {class=OverlaySlot, properties {horizontalAlignment=HAlign_Fill, verticalAlignment=VAlign_Fill, padding=<ltrb -8 -8 -8 -8>}}
  }
  FortResponsivePanel_1 {
   widgetClassPath=/Script/FortniteUI.FortResponsivePanel
   parent=Overlay_1
   widget=FortResponsivePanel_1
   slot {class=OverlaySlot, properties {horizontalAlignment=HAlign_Fill}}
  }
  SizeBox_1 {
   widgetClassPath=&SizeBox
   parent=FortResponsivePanel_1
   widget=SizeBox_1
   properties {bOverride_MinDesiredHeight=true, bOverride_MinDesiredWidth=true, minDesiredHeight=50, minDesiredWidth=200}
   slot {class=FortResponsivePanelSlot}
  }
  Panel_Value {widgetClassPath=&Overlay, bIsVariable=true, parent=SizeBox_1, widget=Panel_Value, slot {class=SizeBoxSlot}}
  Progress {
   widgetClassPath=/Game/Athena/HUD/SimpleMaterialProgressBar
   bIsVariable=true
   parent=Panel_Value
   widget=Progress
   properties {
    bOverrideDefaultColorA=true
    bOverrideDefaultColorB=true
    bOverrideDefaultColorBackground=true
    visibility=HitTestInvisible
    cachedColorA=<rgba 0.0382 0.4452 0.9047 1>
    cachedColorB=<rgba 0.0382 0.4452 0.9047 1>
    cachedColorBackground=<rgba 0.0033 0.0097 0.1022 1>
   }
   slot {class=OverlaySlot, properties {horizontalAlignment=HAlign_Fill, verticalAlignment=VAlign_Fill}}
  }
  Text_SettingValue {
   widgetClassPath=&CommonTextBlock
   bIsVariable=true
   parent=Panel_Value
   widget=Text_SettingValue
   properties {
    justification=Center
    scrollStyle=&TextScrollStyle-Base_C
    style=/FortUILibrary/Styles/Typography/Foundation/TextStyle-Base-S-B.TextStyle-Base-S-B_C
    text="{Value}"
    visibility=HitTestInvisible
    font {
     bForceMonospaced=false
     bMaterialIsStencil=false
     fontMaterial=null
     fontObject=&NotoSans
     letterSpacing=0
     monospacedWidth=1
     size=18
     skewAmount=0
     typefaceFontName=NotoSans-Bold
     outlineSettings {
      bApplyOutlineToDropShadows=false
      bMiteredCorners=false
      bSeparateFillAlpha=false
      outlineMaterial=null
      outlineSize=0
      outlineColor=<rgba 0 0 0 1>
     }
    }
    shadowOffset=<0 0>
   }
   slot {class=OverlaySlot, properties {horizontalAlignment=HAlign_Fill, verticalAlignment=VAlign_Center}}
  }
  Slider_SettingValue {
   widgetClassPath=/Script/CommonUI.AnalogSlider
   bIsVariable=true
   parent=Panel_Value
   widget=Slider_SettingValue
   properties {sliderBarColor=<rgba 1 1 1 0>, sliderHandleColor=<rgba 1 1 1 0>}
   slot {class=OverlaySlot, properties {horizontalAlignment=HAlign_Fill, verticalAlignment=VAlign_Fill}}
  }
 ]
}
