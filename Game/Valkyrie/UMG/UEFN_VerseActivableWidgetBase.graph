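# Base activatable widget for UEFN Verse UI, overlaying an inherited named content slot with a hamburger button.
# While active, it hides tagged HUD elements on the owning Fort player controller, restoring them on deactivation.
# It tracks input method changes to collapse the hamburger button on gamepad and show it for mouse or touch.
# Desired focus goes to the slot's content, falling back to itself, and it reports that cursor mode changes don't clear it.
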WidgetBlueprint /Game/Valkyrie/UMG/UEFN_VerseActivableWidgetBase
graphs [
 BP_GetDesiredFocusTarget {nodes [
   Get Desired Focus Target()
   n1 = Get MainSlot() -> (MainSlot:Verse Fortnite UINamed Slot Object Reference)
   > n2 = Is Valid(InputObject=@n1)
   Return Node!(execute=[@n4.Is Not Valid, @n2.Is Not Valid], ReturnValue:Widget Object Reference=Self-Reference())
   n3 = GetContent!(self:Content Widget Object Reference=@n1) -> (ReturnValue:Widget Object Reference)
   > n4 = Is Valid(InputObject=@n3)
   > Return Node(ReturnValue=@n3)
  ]}
 EventGraph {nodes [
   Event On Activated()
   n1 = Create Event()
   n2 = Get CommonInputSubsystem()
   > Bind Event to On Input Method Changed(self=@n2, Delegate=@n1)
   > OnInputMethodChanged!(execute, self, NewInputType:ECommonInputType Enum=GetCurrentInputType(self=@n2)) -> (then)
   n3 = GetOwningPlayer()
   > n4 = Cast To FortPlayerController(Object=@n3)
   n5 = Get HUDElementsToHide() -> (HUDElementsToHide:Gameplay Tag Container Structure)
   n6 = Get HUDElementReason() -> (HUDElementReason:Name=None)
   > SetHUDElementVisibilityWithContext!(
    execute
    self:Fort Player Controller Object Reference=@n4
    HUDElementTags:"Gameplay Tag Container Structure (by ref)"=@n5
    bHideElements:bool=true
    ReasonName:Name=@n6
    LogContext:String
   ) -> (then)
   Event On Deactivated()
   > Unbind Event from On Input Method Changed(self=@n2, Delegate=@n1)
   > n7 = Cast To FortPlayerController(Object=@n3)
   > SetHUDElementVisibilityWithContext(self=@n7, HUDElementTags=@n5, bHideElements=false, ReasonName=@n6)
  ]}
 IsClearedWhenCursorModeChanges {nodes [
   IsClearedWhenCursorModeChanges() {is=title K2Node_FunctionEntry}
   > Return Node(ReturnValue:bool=false)
  ]}
 OnInputMethodChanged {nodes [
   n1 = OnInputMethodChanged!() -> (then, NewInputType:ECommonInputType Enum) {is=title K2Node_FunctionEntry}
   > SetVisibility(
    self=Get HamburgerButton() -> (HamburgerButton:Hamburger Button Object Reference)
    InVisibility=Select!(
     MouseAndKeyboard:ESlateVisibility Enum=Visible
     Gamepad:ESlateVisibility Enum=Collapsed
     Touch:ESlateVisibility Enum=Visible
     Count:ESlateVisibility Enum=Collapsed
     Index:ECommonInputType Enum=@n1
    ) -> (ReturnValue)
   )
  ]}
]
variables [
 HUDElementReason {type=string, default=UEFNVerseWidget}
 HUDElementsToHide {type=GameplayTagContainer, default {gameplayTags [{tagName=HUD.Athena.HideforUEFNVerseWidget}]}}
]
widgets {
 info {
  inheritedWidgetCount=1
  parentClass=/Script/VerseFortniteUI.VerseFortniteUIActivableWidgetBase
  rootWidgetClass=&Overlay
  widgetCount=3
 }
 widgets [
  Overlay_0 {widgetClassPath=&Overlay, widget=Overlay_0}
  MainSlot {
   widgetClassPath=/Script/VerseFortniteUI.VerseFortniteUINamedSlot
   bInherited=true
   parent=Overlay_0
   widget=MainSlot
   slot {class=OverlaySlot, properties {horizontalAlignment=HAlign_Fill, verticalAlignment=VAlign_Fill}}
  }
  HamburgerButton {
   widgetClassPath=/Game/Athena/HUD/Bacchus/CustomizableHUD/HamburgerButton
   bIsVariable=true
   parent=Overlay_0
   widget=HamburgerButton
   properties {
    elementTag {tagName=HUD.Bacchus.Hamburger.UEFN}
    hUDElementTag {gameplayTags [{tagName=HUD.Athena.HideForPostGame}, {tagName=HUD.Athena.HideForPostGameSpectate}]}
   }
   slot {class=OverlaySlot, properties {horizontalAlignment=HAlign_Right, padding=<ltrb 50 20 20 50>}}
  }
 ]
}
